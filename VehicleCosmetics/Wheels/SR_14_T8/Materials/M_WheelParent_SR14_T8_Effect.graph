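# Translucent effect material for the SR_14_T8 vehicle wheel, usable on skeletal meshes, producing an animated electric glow.
# Two opposing panned cloud-noise samples combine to warp a panning gradient texture.
# That result and the noise offset a rotated, tiled second gradient sample, which the inverted first gradient masks.
# The contrast-boosted streaks brighten a tinted glow color, with parameters for speeds, electricity intensity, white center and glow strength.
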Material /VehicleCosmetics/Wheels/SR_14_T8/Materials/M_WheelParent_SR14_T8_Effect
graph {
 properties {
  bAutomaticallySetUsageInEditor=false
  bUsedWithSkeletalMesh=true
  blendMode=BLEND_Translucent
  previewMesh=/VehicleCosmetics/Wheels/SR_14_T8/Meshes/SK_Wheel_SR_14_T8
  usedShadingModels=Default Lit
 }
 nodes [
  n1 = Time()
  n2 = ScalarParameter() {defaultValue=1, parameterName=PanningSpeed}
  n3 = Multiply(A=Time(), B=ScalarParameter() {defaultValue=1, parameterName=ElectricitySpeed})
  n4 = TextureSample(UVs=Panner(Coordinate=TextureCoordinate() {uTiling=0.4, vTiling=0.4}, Time=@n3) {speedX=-0.2, speedY=-0.2}) {texture=/VehicleCosmetics/Wheels/_SharedTextures/FX/DCU_CloudDistort_Skin_T}
  n5 = TextureSample(UVs=Panner(Coordinate=TextureCoordinate() {uTiling=0.2, vTiling=0.2}, Time=@n3) {speedX=0.2, speedY=0.2}) {texture=&DCU_CloudDistort_Skin_T}
  n6 = Clamp(None=ConstantBiasScale(None=Add(A=@n4.R, B=@n5.R)) {bias=-0.5, scale=1})
  n7 = Add(
   A=Panner(Coordinate=TextureCoordinate() {uTiling=0.5, vTiling=0.5}, Time=Multiply(A=@n1, B=@n2)) {speedX=0.5, speedY=0.4}
   B=Multiply(A=Constant() {r=0.3}, B=@n6)
  )
  n8 = Clamp(None=ConstantBiasScale(None=TextureSample(UVs=@n7) {texture=/VehicleCosmetics/Wheels/SR_14_T8/Textures/wheel_sr_14_Gradiant_V}.R) {bias=-0.5, scale=3})
  n9 = Add(A=Multiply(A=@n8, B=Constant() {r=0.5}), B=Multiply(A=@n6, B=ScalarParameter() {defaultValue=0.2, parameterName=ElectricityIntencity}))
  n10 = MaterialFunctionCall(UVs=TextureCoordinate() {uTiling=2, vTiling=5}, "Rotation Angle (0-1)"=Constant() {r=-0.25}) {materialFunction=&CustomRotator}
  n11 = Multiply(
   A=OneMinus(None=@n8)
   B=TextureSample(UVs=Add(A=@n9, B=@n10.Rotated Values)) {texture=&wheel_sr_14_Gradiant_V}.R
  )
  n12 = Clamp(None=ConstantBiasScale(None=@n11) {bias=-0.4, scale=5})
  n13 = ScalarParameter() {defaultValue=0.3, parameterName=WhiteCenter_Intensity}
  Multiply(
   A=Add(A=Multiply(A=@n12, B=@n13), B=VectorParameter() {parameterName=GlowColor, defaultValue=<rgba 0.9131 0 0.8714 1>}.RGB)
   B=ScalarParameter() {defaultValue=1.5, parameterName=GlowIntensity}
  )
  Reroute(None=@n11)
 ]
}
